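# SolidColorFloor is a Default Lit material, named as a plain colored floor and stored among the Asteria tree foliage master material functions.
# It exposes a Color vector parameter, defaulting to dark olive green, and a Roughness scalar whose default sits far above the usual 0–1 range.
# The graph holds only these two parameter nodes, with no textures, math or visible wiring to material outputs.
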Material /Game/Environments/Asteria/Foliage/Trees/MasterMaterials/MaterialFunctions/SolidColorFloor
graph {
 properties {pixelDepthOffsetMode=PDOM_Legacy, usedShadingModels=Default Lit}
 nodes [
  ScalarParameter() {defaultValue=11, parameterName=Roughness}
  VectorParameter() {parameterName=Color, defaultValue=<rgba 0.0578 0.1042 0.0032 1>}
 ]
}
