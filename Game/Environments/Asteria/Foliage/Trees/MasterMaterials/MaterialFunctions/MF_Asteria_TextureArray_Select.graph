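# Material function for Asteria tree foliage materials that turns 2D UVs into three-component UVs for texture array sampling.
# It takes a Vector2 input, defaulting to the texture coordinate, and appends one constant as the slice index.
# The general, Leaves_Unique and Trunk_Unique outputs all build the same vector.
# The constant is marked "Removed" and left at its default, so no per-part or variable atlas selection takes place.
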MaterialFunction /Game/Environments/Asteria/Foliage/Trees/MasterMaterials/MaterialFunctions/MF_Asteria_TextureArray_Select
graph {nodes [
  n1 = FunctionInput(Preview=TextureCoordinate()) {bUsePreviewValueAsDefault=true, inputType=FunctionInput_Vector2}
  n2 = Constant() {desc=Removed}
  FunctionOutput(None=AppendVector(A=@n1, B=@n2)) {outputName="UVs(vector3)WithAtlasSelection"}
  FunctionOutput(None=AppendVector(A=@n1, B=@n2)) {outputName="UVs(vector3)WithAtlasSelection_Leaves_Unique"}
  FunctionOutput(None=AppendVector(A=@n1, B=@n2)) {outputName="UVs(vector3)WithAtlasSelection_Trunk_Unique"}
 ]}
