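# Material function for Asteria tree master materials that gates scalar and vector attributes behind the UseTextureArrays and UseUDIM_Textures static switches.
# With both switches off it passes the float and RGB inputs through; either switch on outputs zero constants instead.
# An Index-based UV into a 128-wide lookup feeds MF_Asteria_TextureArray_Select, but that call's result reaches neither output.
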MaterialFunction /Game/Environments/Asteria/Foliage/Trees/MasterMaterials/MaterialFunctions/MF_AsteriaTree_ArrayAttribute
graph {nodes [
  n1 = FunctionInput() {inputName=Float In, inputType=FunctionInput_Scalar}
  n2 = Constant()
  n3 = Constant()
  FunctionOutput(None=StaticSwitchParameter(True=@n3, False=StaticSwitchParameter(True=@n2, False=@n1) {group=TextureArray, parameterName=UseUDIM_Textures}) {group=TextureArray, parameterName=UseTextureArrays}) {outputName=Float Out}
  n4 = ComponentMask(None=FunctionInput() {inputName=Vector_In, sortPriority=1}) {b=true, g=true, r=true}
  FunctionOutput(None=StaticSwitchParameter(True=@n3, False=StaticSwitchParameter(True=@n3, False=@n4) {group=TextureArray, parameterName=UseUDIM_Textures}) {group=TextureArray, parameterName=UseTextureArrays}) {outputName=Vector_Out, sortPriority=1}
  n5 = Divide(A=Constant() {r=1}, B=Constant() {desc=TextureWidth, r=128})
  n6 = Add(A=FunctionInput() {inputName=Index, inputType=FunctionInput_Scalar, sortPriority=3}, B=Multiply(A=@n5) {constB=0.5}) {constB=0.5}
  MaterialFunctionCall!(In=AppendVector(A=Multiply(A=@n5, B=@n6), B=Constant() {r=0.5})) {materialFunction=&/MF_Asteria_TextureArray_Select}
 ]}
