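# Material function for Asteria trees that blends ground moss, or landscape RVT color, over incoming material attributes, skipped at Low and Medium quality.
# A height lerp shapes the mask from vertex-normal up-facing and distance to an instance-space origin.
# An optional RVT landscape-height sphere mask serves very wide trunks.
# Static switches and material proxy replacement fall back to plain moss for impostor bakes, proxies or disabled RVT.
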MaterialFunction /Game/Environments/Asteria/Foliage/Trees/MasterMaterials/MaterialFunctions/MF_Asteria_Tree_RVTBlend
graph {nodes [
  n1 = FunctionInput() {inputType=FunctionInput_MaterialAttributes}
  n2 = BlendMaterialAttributes(A=@n1, B=NamedRerouteUsage(), Alpha=NamedRerouteUsage()) {vertexAttributeBlendType=UseA}
  FunctionOutput(None=StaticSwitchParameter(True=QualitySwitch(Default=@n2, Low=@n1, High=@n2, Medium=@n1, Epic=@n2), False=@n1) {defaultValue=true, group=Moss, parameterName=UseRVTMossBlend})
  n3 = TextureSampleParameter2D() {
   group=Moss
   parameterName=Moss_BaseColor
   samplerSource=SSM_Wrap_WorldGroupSettings
   texture=/Game/Environments/Apollo/Rocks/RockSetA/Textures/T_RockSetA_Moss_01_D
  }
  n4 = VectorParameter() {group=Moss, parameterName=TintColor_Moss, channelNames {a=Desaturation, b="", g="", r=""}, defaultValue=<rgba 1 1 1 0>}
  n5 = TextureSampleParameter2D() {
   group=Moss
   parameterName=Moss_Normal
   samplerSource=SSM_Wrap_WorldGroupSettings
   samplerType=SAMPLERTYPE_Normal
   texture=/Game/Environments/Apollo/Rocks/RockSetA/Textures/T_RockSetA_Moss_01_N
  }
  n6 = SetMaterialAttributes!(
   MaterialAttributes
   BaseColor=Multiply(A=Desaturation(None=@n3.RGB, Fraction=@n4.Desaturation), B=@n4.RGB)
   Normal=@n5.RGB
   Specular=ScalarParameter() {defaultValue=0.5478, group=Moss, parameterName=Spec_Moss}
   Roughness=ScalarParameter() {defaultValue=0.9738, group=Moss, parameterName=Roughnes_Moss}
   Metallic
  ) {attributeSetTypes [
    #1
    #2
    #3
    #4
    #5
   ]}
  n7 = NamedRerouteUsage()
  n8 = MaterialProxyReplace(
   Realtime=MaterialFunctionCall(NoRVT=@n6, RVT=SetMaterialAttributes!(MaterialAttributes=@n7, Normal=@n5.RGB) {attributeSetTypes [#2]}) {materialFunction=&FN_RVT_Switch_Material}.Result
   MaterialProxy=@n6
  )
  n9 = StaticBool()
  NamedRerouteDeclaration(None=StaticSwitch(
    True=@n6
    False=StaticSwitchParameter(True=@n6, False=@n8) {group=RVT, parameterName=Disable_RVT}
    Value=FunctionInput(Preview=@n9) {bUsePreviewValueAsDefault=true, inputName=IsImpostorBake, inputType=FunctionInput_StaticBool, sortPriority=20}
   )) {name=MossMaterial, nodeColor=<rgba 0.1719 0 1 1>}
  NamedRerouteDeclaration(None=@n3.A) {name=MossHeightMask, nodeColor=<rgba 0 1 0.6719 1>}
  n10 = NamedRerouteUsage()
  n11 = Multiply(A=NamedRerouteUsage(), B=ScalarParameter() {defaultValue=3.6084, group=Moss, parameterName=MossHeightAdd})
  NamedRerouteDeclaration(None=MaterialFunctionCall(
    Transition Phase=NamedRerouteUsage()
    Height Texture=Add(A=@n10, B=@n11)
    Contrast=ScalarParameter() {defaultValue=0.538, group=Moss, parameterName=MossTransition}
   ) {materialFunction=&HeightLerp}.Alpha) {name=MossBlendMask, nodeColor=<rgba 0 0.2266 1 1>}
  n12 = ComponentMask(None=VertexNormalWS()) {b=true}
  n13 = ScalarParameter() {defaultValue=1.31, group=Moss, parameterName=MossUPContrast}
  n14 = VectorParameter() {desc=position relative to pivot that the gradient will originate from, group=Moss, parameterName=MossOrigin, defaultValue=<rgba 0 0 0 1>}
  n15 = Distance(
   A=WorldPosition().XYZ
   B=TransformPosition(None=@n14.RGB) {transformSourceType=TRANSFORMPOSSOURCE_Instance, transformType=TRANSFORMPOSSOURCE_World}
  )
  n16 = ScalarParameter() {defaultValue=500, group=Moss, parameterName=GroundMossDistance}
  n17 = Saturate(None=OneMinus(None=Divide(A=@n15, B=@n16)))
  n18 = OneMinus(None=ScalarParameter() {defaultValue=1, desc=weighting of vertex normal based mask, group=Moss, parameterName=MossUPBlend})
  n19 = LinearInterpolate(
   A=Multiply(
    A=MaterialFunctionCall(In=@n12, Contrast=@n13) {materialFunction=&CheapContrast}.Result
    B=@n17
   )
   B=@n17
   Alpha=Saturate(None=Add(A=@n17, B=@n18))
  )
  n20 = Saturate(None=@n19)
  n21 = NamedRerouteUsage()
  n22 = StaticSwitchParameter(
   True=MaterialProxyReplace(
    Realtime=MaterialFunctionCall!(NoRVT=@n20, RVT=@n21) {materialFunction=/Game/Packages/Fortress_SharedMaterials/Base_Material_Functions/FN_RVT_Switch_Scalar}.Result
    MaterialProxy=@n20
   )
   False=@n20
  ) {
   desc="Only use for special cases with very large/ wide trunks that cover a larger portion of the landscape and need more accurate representation of the ground than the standard masking options offer!"
   group=Moss
   parameterName=AllowRVTHeightBlend
  }
  NamedRerouteDeclaration(None=StaticSwitchParameter(True=@n20, False=@n22) {defaultValue=true, group=RVT, parameterName=Disable_RVT}) {name=Transition, nodeColor=<rgba 0 0.5781 1 1>}
  n23 = RuntimeVirtualTextureSampleParameter() {materialType=WorldHeight, parameterName=RVTHeight, virtualTexture=/Game/Athena/Apollo/Maps/Landscape/VT/RVT_Apollo_Landscape_Height}
  n24 = ComponentMask(None=WorldPosition().XYZ) {b=true}
  n25 = ScalarParameter() {desc=offsets height of rvt blend, group=RVT, parameterName=Blend RVT Offset}
  n26 = ObjectRadius()
  n27 = VertexNormalWS()
  n28 = LinearInterpolate(
   A=ScalarParameter() {defaultValue=0.132, group=RVT, parameterName=Blend RVT North}
   B=ScalarParameter() {defaultValue=0.056, group=RVT, parameterName=Blend RVT}
   Alpha=Saturate(None=ComponentMask(None=@n27) {r=true})
  )
  n29 = Saturate(None=SphereMask(
    A=@n23.WorldHeight
    B=Add(A=@n24, B=@n25)
    Radius=Multiply(A=@n26, B=@n28)
    Hardness=ScalarParameter() {defaultValue=-1.4114, group=RVT, parameterName=BlendRVTContrast}
   ))
  NamedRerouteDeclaration(None=@n29) {name=RVT Heightblend, nodeColor=<rgba 1 0 0.2344 1>}
  n30 = RuntimeVirtualTextureSampleParameter!(UVs, World Position, Mip Level=Constant() {r=8}) {
   bEnableFeedback=false
   materialType=BaseColor_Normal_Specular_YCoCg
   mipValueMode=RVTMVM_MipLevel
   parameterName=RVTColor
   virtualTexture=/Game/Athena/Apollo/Maps/Landscape/VT/RVT_Apollo_Landscape
  }
  NamedRerouteDeclaration(None=SetMaterialAttributes!(MaterialAttributes, BaseColor=@n30.BaseColor, Specular=@n30.BaseColor, Roughness=@n30.BaseColor) {attributeSetTypes [#1, #3, #4]}) {name=RVT Material, nodeColor=<rgba 0.9922 1 0 1>}
  Reroute(None=@n19)
  Reroute(None=@n29)
 ]}
