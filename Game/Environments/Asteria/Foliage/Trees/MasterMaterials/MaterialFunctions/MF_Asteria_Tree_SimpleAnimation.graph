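# Material function that adds simple wind animation to Asteria tree materials by setting World Position Offset, behind a static switch.
# MF_WindActorData, sampled with per-leaf vertex-color offsets, drives vertex-color-masked displacement and rotation around axes and pivots stored in UVs.
# An optional trunk bend adds height-masked sine/cosine sway with a random phase, blending default and storm settings.
# The bend rescales vectors to their original length, so vertices arc instead of stretching.
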MaterialFunction /Game/Environments/Asteria/Foliage/Trees/MasterMaterials/MaterialFunctions/MF_Asteria_Tree_SimpleAnimation
graph {nodes [
  n1 = WorldPosition()
  n2 = AppendVector(A=Constant2Vector(), B=VertexColor().R)
  n3 = ScalarParameter() {defaultValue=500, group=SimpleAnimation, parameterName=SimpleAnimationRandomPerLeafOffset}
  n4 = MaterialFunctionCall!(
   Position=Add(A=WorldPosition().XYZ, B=Multiply(A=@n2, B=@n3))
   IsTrees
   IsGrass
   IsMediumPlant
   IsMediumPlant_2024Anim
   IsMediumPlant2ndLevel
   MediumPlant_BasePivotPostion
  ) {materialFunction=/Game/Environments/Asteria/Foliage/Wind/MF_WindActorData}
  n5 = Multiply(A=@n4.WindI_Direction, B=ScalarParameter() {defaultValue=10, group=SimpleAnimation, parameterName=SimpleAnimationMaxOffset})
  n6 = VertexColor()
  n7 = TextureCoordinate() {coordinateIndex=2}
  n8 = ComponentMask(None=@n7) {g=true}
  n9 = TextureCoordinate() {coordinateIndex=3}
  n10 = CrossProduct(A=Transform(None=AppendVector(A=@n8, B=@n9)) {transformSourceType=TRANSFORMSOURCE_Instance}, B=@n4.WindI_Direction)
  n11 = ScalarParameter() {defaultValue=0.784, group=SimpleAnimation, parameterName=SimpleAnimationMaxRotation}
  n12 = Multiply(
   A=VertexColor().B
   B=RotateAboutAxis(
    NormalizedRotationAxis=Normalize(VectorInput=@n10)
    RotationAngle=Multiply(A=@n4.WindIntensitySingleSample, B=@n11)
    PivotPoint=NamedRerouteUsage()
    Position=WorldPosition().XYZ
   )
  )
  n13 = Add(A=Multiply(A=@n4.WindIntensitySingleSample, B=Multiply(A=@n5, B=@n6.G)), B=@n12)
  Add(A=@n1.XYZ, B=@n13)
  n14 = FunctionInput() {inputType=FunctionInput_MaterialAttributes}
  n15 = GetMaterialAttributes(None=@n14) {attributeGetTypes [#1]}
  n16 = NamedRerouteUsage()
  n17 = Subtract(A=@n16, B=@n1.XYZ)
  n18 = TransformPosition(None=Constant3Vector()) {transformSourceType=TRANSFORMPOSSOURCE_Instance, transformType=TRANSFORMPOSSOURCE_World}
  n19 = Subtract(
   A=ComponentMask(None=WorldPosition().XYZ) {b=true}
   B=ComponentMask(None=NamedRerouteDeclaration(None=@n18) {name=InstancePosition, nodeColor=<rgba 1 0.0469 0 1>}) {b=true}
  )
  n20 = ScalarParameter() {group=SimpleAnimation, parameterName=TrunkBendOffset}
  n21 = ObjectRadius()
  n22 = Saturate(None=Divide(A=Subtract(A=@n19, B=@n20), B=Multiply(A=@n21) {constB=2}))
  n23 = Time()
  n24 = MaterialFunctionCall(Float3=NamedRerouteUsage()) {materialFunction=&BreakOutFloat3Components}
  n25 = VertexColor()
  n26 = Divide(A=Add(A=Add(A=@n24.R, B=@n24.R), B=Multiply(A=@n25.A) {constB=33.33})) {constB=10}
  n27 = NamedRerouteDeclaration(None=Multiply(A=Frac(None=@n26)) {constB=23.33}) {name=RandomNumber, nodeColor=<rgba 0 0.1797 1 1>}
  n28 = Add(A=Multiply(A=@n23, B=ScalarParameter() {defaultValue=0.25, group=Animation WInd, parameterName=Wind_Bend_Speed_Default}), B=@n27)
  n29 = Add(A=Multiply(A=@n23, B=ScalarParameter() {defaultValue=1, group=Animation WInd, parameterName=Wind_Bend_Speed_Storm}), B=@n27)
  n30 = NamedRerouteUsage()
  n31 = LinearInterpolate(A=Sine(None=@n28), B=Sine(None=@n29), Alpha=@n30)
  n32 = LinearInterpolate(A=Cosine(None=Multiply(A=@n28) {constB=1.13}), B=Cosine(None=Multiply(A=@n29) {constB=1.13}), Alpha=@n30)
  n33 = Multiply(
   A=MaterialFunctionCall(X=Add(A=@n31) {constB=0.75}, Y=Multiply(A=@n32) {constB=0.618}) {materialFunction=&MakeFloat3}.Result
   B=LinearInterpolate(
    A=ScalarParameter() {defaultValue=8, group=Animation WInd, parameterName=Wind_Bend_Strength}
    B=ScalarParameter() {defaultValue=40, group=Animation WInd, parameterName=Wind_Bend_Strength_Storm}
    Alpha=@n30
   )
  )
  n34 = Subtract(A=@n16, B=Add(A=@n1.XYZ, B=Multiply(A=@n22, B=@n33)))
  n35 = Divide(
   A=@n34
   B=Divide(
    A=MaterialFunctionCall(Vector 3=@n34) {materialFunction=/Engine/Functions/Engine_MaterialFunctions02/Utility/VectorLength}.V3 Length
    B=MaterialFunctionCall(Vector 3=@n17) {materialFunction=&VectorLength}.V3 Length
   )
  )
  n36 = StaticSwitchParameter(True=Add(A=Subtract(A=@n17, B=@n35), B=@n13), False=@n13) {group=SimpleAnimation, parameterName=UseSimpleAnimationTrunkMovement}
  FunctionOutput(None=StaticSwitchParameter(
    True=SetMaterialAttributes!(MaterialAttributes=@n15.MaterialAttributes, WorldPositionOffset=@n36) {attributeSetTypes [#1]}
    False=@n14
   ) {group=SimpleAnimation, parameterName=UseSimpleAnimation})
  n37 = TextureCoordinate() {coordinateIndex=1}
  n38 = ComponentMask(None=@n7) {r=true}
  NamedRerouteDeclaration(None=TransformPosition(None=AppendVector(A=@n37, B=@n38)) {transformSourceType=TRANSFORMPOSSOURCE_Instance, transformType=TRANSFORMPOSSOURCE_World}) {name=PivotPosition, nodeColor=<rgba 0.5469 0 1 1>}
  NamedRerouteDeclaration(None=@n4.StormMask) {name=StormMask, nodeColor=<rgba 0.7344 1 0 1>}
  NamedRerouteUsage()
  NamedRerouteUsage()
  Reroute(None=@n18)
  n39 = TransformPosition(None=Constant3Vector()) {transformSourceType=TRANSFORMPOSSOURCE_Instance, transformType=TRANSFORMPOSSOURCE_World}
  n40 = MaterialFunctionCall(Position=@n39) {materialFunction=&MF_WindActorData}
  n41 = Divide(
   A=Distance(A=@n39, B=WorldPosition().XYZ)
   B=ScalarParameter() {defaultValue=150, group=SimpleAnimation, parameterName=SimpleAnimationObjectHeight}
  )
  n42 = Multiply(
   A=Multiply(A=@n4.WindI_Direction, B=Saturate(None=@n41))
   B=ScalarParameter() {defaultValue=10, group=SimpleAnimation, parameterName=SimpleAnimationMaxOffset}
  )
  StaticSwitchParameter(True=Add(A=@n12, B=Multiply(A=@n40.WindIntensitySingleSample, B=@n42)), False=@n12) {group=SimpleAnimation, parameterName=UseSimpleAnimation}
 ]}
