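# Material function for Asteria trees that varies leaf and trunk color using vertex colors.
# Leaves get a blue-masked brightness and hue shift, optionally driven by per-instance or primitive custom data, and bypassed by the UseSimpleAnimation switch.
# Trunk gets blue-masked contrast, tint and desaturation, red-scaled hue shift, then an alpha-masked color blend.
# Most parameters route through indexed MF_AsteriaTree_ArrayAttribute calls; two vertex-color switches sit unconnected.
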MaterialFunction /Game/Environments/Asteria/Foliage/Trees/MasterMaterials/MaterialFunctions/MF_Asteria_Tree_VertexColorBlend
graph {nodes [
  n1 = MaterialFunctionCall!(
   Float In=ScalarParameter() {group=Color VertexColor Tints, parameterName=Blue Vertex Channel Hueshift, sortPriority=12}
   Vector_In
   Index=Constant() {r=14}
  ) {materialFunction=&/MF_AsteriaTree_ArrayAttribute}
  n2 = ScalarParameter() {bUseCustomPrimitiveData=true, defaultValue=0.5, group=Color, parameterName=ValueVariation_primdata, primitiveDataIndex=5}
  n3 = StaticSwitchParameter(True=PerInstanceCustomData!(DefaultValue) {constDefaultValue=0.5, dataIndex=5}, False=@n2) {parameterName=UsePerInstanceCustomData}
  n4 = FunctionInput(Preview=Constant()) {bUsePreviewValueAsDefault=true, inputName=UseCustomPrimValue, inputType=FunctionInput_Scalar, sortPriority=200}
  n5 = MaterialFunctionCall(
   Float In=ScalarParameter() {defaultValue=1, group=Color VertexColor Tints, parameterName=Blue Vertex Channel Tint Brightness, sortPriority=12}
   Index=Constant() {r=13}
  ) {materialFunction=&MF_AsteriaTree_ArrayAttribute}
  n6 = FunctionInput(Preview=Constant3Vector() {constant=<rgba 0.8792 0.1535 0.8171 0>}) {inputName=Leaves Input}
  n7 = MaterialFunctionCall(
   Float In=ScalarParameter() {group=Color VertexColor Tints, parameterName=Blue Vertex Channel Mask Value LOW, sortPriority=12}
   Index=Constant() {r=15}
  ) {materialFunction=&MF_AsteriaTree_ArrayAttribute}
  n8 = MaterialFunctionCall(
   Float In=ScalarParameter() {defaultValue=1, group=Color VertexColor Tints, parameterName=Blue Vertex Channel Mask Value HIGH, sortPriority=12}
   Index=Constant() {r=16}
  ) {materialFunction=&MF_AsteriaTree_ArrayAttribute}
  n9 = VertexColor()
  n10 = LinearInterpolate(
   A=@n6
   B=MaterialFunctionCall(Hue Shift Percentage=LinearInterpolate(A=@n1.Float Out, B=@n3, Alpha=@n4), Texture=Multiply(A=@n6, B=@n5.Float Out)) {materialFunction=&HueShift}.Result
   Alpha=Saturate(None=LinearInterpolate(A=@n7.Float Out, B=@n8.Float Out, Alpha=@n9.B))
  )
  FunctionOutput(None=StaticSwitchParameter(True=@n6, False=@n10) {group=SimpleAnimation, parameterName=UseSimpleAnimation}) {outputName=Leaves Output}
  n11 = Multiply(A=ScalarParameter() {group=Trunk Color VertexColor, parameterName=Red Vertex Color Trunk Hue Shift, sortPriority=12}, B=VertexColor().R)
  n12 = FunctionInput() {inputName=Trunk_Input, sortPriority=3}
  n13 = MaterialFunctionCall!(
   In=@n12
   Contrast=MaterialFunctionCall(
    Float In=ScalarParameter() {group=Trunk Color VertexColor, parameterName=Blue Vertex Color Trunk Contrast, sortPriority=12}
    Index=Constant() {r=37}
   ) {materialFunction=&MF_AsteriaTree_ArrayAttribute}.Float Out
  ) {materialFunction=/Engine/Functions/Engine_MaterialFunctions01/ImageAdjustment/CheapContrast_RGB}
  n14 = VectorParameter() {group=Trunk Color VertexColor, parameterName=Blue Vertex Color Trunk Tint, sortPriority=26, defaultValue=<rgba 1 1 1 1>}
  n15 = MaterialFunctionCall(Vector_In=@n14.RGB, Index=Constant() {r=38}) {materialFunction=&MF_AsteriaTree_ArrayAttribute}
  n16 = ScalarParameter() {group=Trunk Color VertexColor, parameterName=Blue Vertex Color Trunk Desaturation, sortPriority=12}
  n17 = Constant() {r=39}
  n18 = MaterialFunctionCall(
   Float In=ScalarParameter() {group=Trunk Color VertexColor, parameterName=Blue Vertex Color Trunk Mask Value LOW, sortPriority=10}
   Index=Constant() {r=40}
  ) {materialFunction=&MF_AsteriaTree_ArrayAttribute}
  n19 = MaterialFunctionCall(
   Float In=ScalarParameter() {group=Trunk Color VertexColor, parameterName="Blue Vertex Color Trunk  Mask Value HIGH", sortPriority=11}
   Index=Constant() {r=41}
  ) {materialFunction=&MF_AsteriaTree_ArrayAttribute}
  n20 = VertexColor()
  n21 = LinearInterpolate(
   A=@n12
   B=Desaturation(
    None=Multiply(A=@n13.Result, B=@n15.Vector_Out)
    Fraction=MaterialFunctionCall(Float In=@n16, Index=@n17) {materialFunction=&MF_AsteriaTree_ArrayAttribute}.Float Out
   )
   Alpha=Saturate(None=LinearInterpolate(A=@n18.Float Out, B=@n19.Float Out, Alpha=@n20.B))
  )
  n22 = VectorParameter() {group=Trunk Color VertexColor, parameterName="RGB Trunk Blend Color(AlphaChannel)", sortPriority=26, defaultValue=<rgba 1 1 1 1>}
  n23 = Constant() {r=17}
  n24 = LinearInterpolate(
   A=MaterialFunctionCall(
    Float In=ScalarParameter() {group=Trunk Color VertexColor, parameterName="RGB Trunk Blend Mask Value LOW(AlphaChannel)", sortPriority=27}
    Index=Constant() {r=18}
   ) {materialFunction=&MF_AsteriaTree_ArrayAttribute}.Float Out
   B=MaterialFunctionCall(
    Float In=ScalarParameter() {group=Trunk Color VertexColor, parameterName="RGB Trunk Blend Mask Value HIGH(AlphaChannel)", sortPriority=27}
    Index=Constant() {r=19}
   ) {materialFunction=&MF_AsteriaTree_ArrayAttribute}.Float Out
   Alpha=VertexColor().A
  )
  FunctionOutput(None=LinearInterpolate(
    A=MaterialFunctionCall(Hue Shift Percentage=@n11, Texture=@n21) {materialFunction=&HueShift}.Result
    B=MaterialFunctionCall(Vector_In=@n22.RGB, Index=@n23) {materialFunction=&MF_AsteriaTree_ArrayAttribute}.Vector_Out
    Alpha=Saturate(None=@n24)
   )) {outputName=Trunk Output, sortPriority=2}
  StaticSwitchParameter(True=@n10, False=@n6) {group=Color VertexColor Tints, parameterName=Use Blue Channel Vertex Color Hue Shift, sortPriority=10}
  StaticSwitchParameter() {group=Color VertexColor Tints, parameterName=Use RGB Channel Vertex Color Blend, sortPriority=20}
 ]}
