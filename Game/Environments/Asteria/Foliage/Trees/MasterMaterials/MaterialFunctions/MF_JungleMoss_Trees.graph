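# Material function that layers world-aligned jungle moss, with fresnel-tinted color, over incoming tree material attributes.
# Coverage comes from vertex color or large-scale noise times a height gradient from the instance origin, refined by HeightLerp with the moss heightmap.
# Epic and default quality blend detail normals and height-driven roughness and specular, while High uses simpler moss.
# Low, Medium, mobile or a disabled UseJungleMoss switch pass input through unchanged.
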MaterialFunction /Game/Environments/Asteria/Foliage/Trees/MasterMaterials/MaterialFunctions/MF_JungleMoss_Trees
graph {nodes [
  n1 = WorldPosition()
  Add(A=@n1.XYZ, B=VectorParameter() {group=JungleMoss, parameterName=JungleTexture_Large_Offset, defaultValue=<rgba 278.1917 309.5966 0 1>}.RGB)
  n2 = ScalarParameter() {defaultValue=3300, group=JungleMoss, parameterName=JungleMoss_Large_Size_Mask}
  n3 = MaterialFunctionCall(Float3=Divide(A=@n1.XYZ, B=@n2)) {materialFunction=&BreakOutFloat3Components}
  n4 = ScalarParameter() {defaultValue=0.01, group=JungleMoss, parameterName=JungleMoss_Large_Skew}
  n5 = Add(A=@n3.R, B=Multiply(A=@n3.G, B=@n4))
  n6 = Add(A=@n3.B, B=Multiply(A=Add(A=@n3.R, B=@n3.R), B=@n4))
  AppendVector(A=@n5, B=@n6)
  n7 = Add(A=@n3.G, B=Multiply(A=@n3.R, B=@n4))
  AppendVector(A=@n7, B=@n6)
  n8 = MaterialFunctionCall(TextureObject=TextureObject() {texture=/Game/Environments/Asteria/Terrain/Textures/T_Asteria_JungleMoss_H}, TextureSize=@n2) {materialFunction=&WorldAlignedTexture}
  ComponentMask(None=@n8.XYZ Texture) {r=true}
  Constant()
  n9 = GetMaterialAttributes(None=FunctionInput() {inputType=FunctionInput_MaterialAttributes}) {attributeGetTypes [#1, #2]}
  n10 = NamedRerouteUsage()
  n11 = NamedRerouteUsage()
  n12 = QualitySwitch(Default=@n10, Low=@n11, High=@n11, Medium=@n11, Epic=@n10)
  n13 = Constant3Vector()
  n14 = Constant() {r=1}
  n15 = ComponentMask(None=MaterialFunctionCall(TextureObject=TextureObject() {texture=&T_Asteria_JungleMoss_H}, TextureSize=NamedRerouteUsage()) {materialFunction=&WorldAlignedTexture}.XYZ Texture) {r=true}
  n16 = Saturate(None=LinearInterpolate(
    A=ScalarParameter() {defaultValue=0.7259, group=JungleMoss, parameterName=JungleMoss_C3}
    B=ScalarParameter() {defaultValue=1.0799, group=JungleMoss, parameterName=JungleMoss_C4}
    Alpha=@n15
   ))
  n17 = LinearInterpolate(
   A=ScalarParameter() {group=JungleMoss, parameterName=JungleMoss_VertexColor_C1}
   B=ScalarParameter() {defaultValue=1, group=JungleMoss, parameterName=JungleMoss_VertexColor_C2}
   Alpha=ChannelMaskParameter(None=VertexColor().A) {group=JungleMoss, parameterName=VertexColor_Moss}.R
  )
  n18 = Saturate(None=LinearInterpolate(
    A=ScalarParameter() {defaultValue=0.2, group=JungleMoss, parameterName=JungleMoss_Large_C1}
    B=ScalarParameter() {defaultValue=800, group=JungleMoss, parameterName=JungleMoss_Large_C2}
    Alpha=@n8.XY Texture
   ))
  n19 = TransformPosition(None=Constant3Vector()) {transformSourceType=TRANSFORMPOSSOURCE_Instance, transformType=TRANSFORMPOSSOURCE_World}
  n20 = Add(
   A=Subtract(A=WorldPosition().Z, B=ComponentMask(None=@n19) {b=true})
   B=ScalarParameter() {group=JungleMoss, parameterName=JungleMoss_Height_Offset}
  )
  n21 = ScalarParameter() {defaultValue=4057.6423, group=JungleMoss, parameterName=JungleMoss_Height}
  n22 = OneMinus(None=Saturate(None=Divide(A=@n20, B=@n21)))
  n23 = MaterialFunctionCall(
   Transition Phase=StaticSwitchParameter(True=Saturate(None=@n17), False=Multiply(A=@n18, B=@n22)) {group=JungleMoss, parameterName=UseVertexColor}
   Height Texture=@n15
   Contrast=ScalarParameter() {defaultValue=0.448, group=JungleMoss, parameterName=JungleMoss_Contrast}
  ) {materialFunction=&HeightLerp}
  n24 = StaticSwitchParameter(
   True=BlendMaterialAttributes(
    A=@n9.MaterialAttributes
    B=SetMaterialAttributes!(MaterialAttributes=@n12, SubsurfaceColor=@n13, Opacity=@n14) {attributeSetTypes [#3, #4]}
    Alpha=Multiply(A=@n16, B=@n23.Alpha)
   )
   False=@n9.MaterialAttributes
  ) {group=JungleMoss, parameterName=UseJungleMoss}
  FunctionOutput(None=QualitySwitch(
    Default=ShadingPathSwitch(Default=@n24, Mobile=@n9.MaterialAttributes) {desc=Migrated from FeatureLevelSwitch}
    Low=@n9.MaterialAttributes
    Medium=@n9.MaterialAttributes
   ))
  n25 = Add(A=@n3.R, B=@n3.R)
  n26 = Add(A=@n3.G, B=@n3.G)
  Multiply(
   A=Multiply(
    A=TextureSample(UVs=@n25) {samplerSource=SSM_Wrap_WorldGroupSettings, texture=&T_Asteria_JungleMoss_H}.R
    B=TextureSample(UVs=@n26) {samplerSource=SSM_Wrap_WorldGroupSettings, texture=&T_Asteria_JungleMoss_H}.R
   )
   B=TextureSample(UVs=AppendVector(A=@n5, B=@n7)) {samplerSource=SSM_Wrap_WorldGroupSettings, texture=&T_Asteria_JungleMoss_H}.R
  )
  n27 = Fresnel(ExponentIn=ScalarParameter() {defaultValue=0.8607, group=JungleMoss, parameterName=JungleMoss_Fresnell}, Normal=VertexNormalWS()) {baseReflectFraction=0}
  n28 = VectorParameter() {parameterName=JungleMoss_FresnelColor, defaultValue=<rgba 0.5157 1 0.439 1>}
  n29 = ScalarParameter() {defaultValue=-0.172, group=JungleMoss, parameterName=JungleMoss_Hue}
  n30 = NamedRerouteDeclaration(None=ScalarParameter() {defaultValue=500, group=JungleMoss, parameterName=JungleMoss_Size_WS}) {name=MossTextureScale, nodeColor=<rgba 0 1 0.4844 1>}
  n31 = ScalarParameter() {defaultValue=1, group=JungleMoss, parameterName=JungleMoss_PrjectionContrast}
  n32 = MaterialFunctionCall(
   TextureObject=TextureObject() {texture=/Game/Environments/Asteria/Terrain/Textures/T_Asteria_JungleMoss_D}
   TextureSize=@n30
   ProjectionTransitionContrast=@n31
  ) {materialFunction=&WorldAlignedTexture}
  n33 = Add(
   A=Multiply(
    A=Multiply(A=@n27, B=@n28.RGB)
    B=MaterialFunctionCall(Hue Shift Percentage=@n29, Texture=@n32.XYZ Texture) {materialFunction=&HueShift}.Result
   )
   B=Desaturation(None=@n32.XYZ Texture, Fraction=ScalarParameter() {defaultValue=0.088, group=JungleMoss, parameterName=JungleMoss_Desat})
  )
  n34 = MaterialFunctionCall(
   TextureObject=TextureObject() {samplerType=SAMPLERTYPE_Normal, texture=/Game/Environments/Asteria/Terrain/Textures/T_Asteria_JungleMoss_N}
   TextureSize=@n30
   ProjectionTransitionContrast=@n31
  ) {materialFunction=&WorldAlignedNormal}
  n35 = ScalarParameter() {defaultValue=0.5019, group=JungleMoss, parameterName=JungleMoss_Normal}
  n36 = SetMaterialAttributes!(
   MaterialAttributes
   BaseColor=@n33
   Specular=LinearInterpolate(
    A=ScalarParameter() {defaultValue=0.112, group=JungleMoss, parameterName=JungleMoss_Specular_low}
    B=ScalarParameter() {defaultValue=0.5, group=JungleMoss, parameterName=JungleMoss_Specular_high}
    Alpha=@n15
   )
   Roughness=LinearInterpolate(
    A=ScalarParameter() {defaultValue=0.62, group=JungleMoss, parameterName=JungleMoss_Roughness_low}
    B=ScalarParameter() {defaultValue=0.8, group=JungleMoss, parameterName=JungleMoss_Roughness_high}
    Alpha=@n15
   )
   Normal=MaterialFunctionCall(
    BaseNormal=NamedRerouteUsage()
    AdditionalNormal=MaterialFunctionCall(Normal=@n34.XYZ Texture, Flatness=@n35) {materialFunction=&FlattenNormal}.Result
   ) {materialFunction=&BlendAngleCorrectedNormals}.Result
  ) {attributeSetTypes [
    #2
    #5
    #6
    #1
   ]}
  NamedRerouteDeclaration(None=@n36) {name=MossHighEnd, nodeColor=<rgba 1 0.5156 0 1>}
  NamedRerouteDeclaration(None=@n9.Normal) {name=OriginalNormal, nodeColor=<rgba 0 1 0.7188 1>}
  n37 = SetMaterialAttributes(
   BaseColor=@n33
   Roughness=ScalarParameter() {defaultValue=1, group=JungleMoss, parameterName=Mobile__Low_Roughness}
   Normal=NamedRerouteUsage()
  ) {attributeSetTypes [
    #2
    #5
    #6
    #1
   ]}
  NamedRerouteDeclaration(None=@n37) {name=LowEndMossMaterial, nodeColor=<rgba 0 0.9297 1 1>}
  Reroute()
 ]}
