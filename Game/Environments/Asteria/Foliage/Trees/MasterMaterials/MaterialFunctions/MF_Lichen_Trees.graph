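# Tree material function that optionally layers world-aligned lichen over incoming material attributes.
# A mask from world-normal Z, height above the instance origin and optional red vertex paint drives a background tint and height-lerped blend.
# Lichen color is hue-shifted, desaturated and tinted, with optional fuzzy shading and normals blended over the original.
# Low and Medium quality and the mobile shading path pass the input through unchanged.
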MaterialFunction /Game/Environments/Asteria/Foliage/Trees/MasterMaterials/MaterialFunctions/MF_Lichen_Trees
graph {nodes [
  BlendMaterialAttributes()
  n1 = GetMaterialAttributes(None=FunctionInput() {inputType=FunctionInput_MaterialAttributes}) {attributeGetTypes [#1, #2]}
  n2 = GetMaterialAttributes(None=@n1.MaterialAttributes) {attributeGetTypes [#2]}
  n3 = VectorParameter() {group=Lichen, parameterName=Lichen_BackGround_Color, defaultValue=<rgba 0.4581 0.1975 0.9479 1>}
  n4 = ScalarParameter() {group=Lichen, parameterName=Lichen_BackGround_UP_DOWN_Contrast_Low}
  n5 = ScalarParameter() {defaultValue=0.7259, group=Lichen, parameterName=Lichen_BackGround_UP_DOWN_Contrast_High}
  n6 = ComponentMask(None=VertexNormalWS()) {b=true}
  n7 = OneMinus(None=ComponentMask(None=NamedRerouteUsage()) {r=true})
  n8 = ScalarParameter() {defaultValue=6, group=Lichen, parameterName=Vertex Paint Contrast Power, sortPriority=64}
  n9 = Multiply(
   A=Multiply(
    A=Power(Base=@n7, Exp=@n8)
    B=ScalarParameter() {defaultValue=3.5, group=Lichen, parameterName=Vertex Paint Contrast Multiply, sortPriority=64}
   )
   B=VertexColor().R
  )
  n10 = Saturate(None=ScalarParameter() {group=Lichen, parameterName=Use Vertex Painted Lichen?, sortPriority=1})
  n11 = Constant3Vector()
  n12 = Subtract(
   A=WorldPosition().Z
   B=ComponentMask(None=TransformPosition(None=@n11) {transformSourceType=TRANSFORMPOSSOURCE_Instance, transformType=TRANSFORMPOSSOURCE_World}) {b=true}
  )
  n13 = ScalarParameter() {group=Lichen, parameterName=Lichen_Height_Offset}
  n14 = Saturate(None=Divide(A=Add(A=@n12, B=@n13), B=ScalarParameter() {defaultValue=4057.6423, group=Lichen, parameterName=Lichen_Height}))
  n15 = Saturate(None=Add(A=LinearInterpolate(B=@n9, Alpha=@n10), B=OneMinus(None=@n14)))
  n16 = Multiply(A=Saturate(None=LinearInterpolate(A=@n4, B=@n5, Alpha=@n6)), B=@n15)
  n17 = SetMaterialAttributes!(MaterialAttributes=@n2.MaterialAttributes, BaseColor=Multiply(A=@n2.BaseColor, B=LinearInterpolate(B=@n3.RGB, Alpha=@n16) {constA=1})) {attributeSetTypes [#2]}
  n18 = NamedRerouteUsage()
  n19 = NamedRerouteUsage()
  n20 = SetMaterialAttributes!(
   MaterialAttributes=QualitySwitch(Default=@n18, Low=@n19, High=@n19, Medium=@n19, Epic=@n18)
   SubsurfaceColor=Constant3Vector()
   Opacity=Constant() {r=1}
  ) {attributeSetTypes [#3, #4]}
  n21 = GetMaterialAttributes(None=@n20) {attributeGetTypes [#2, #1]}
  n22 = Constant3Vector() {constant=<rgba 0 0 1 1>}
  n23 = ScalarParameter() {defaultValue=1, group=Lichen, parameterName=Lichen_Fuzzy_Normal_Intensity}
  n24 = StaticSwitchParameter(
   True=MaterialFunctionCall!(
    BaseColor=@n21.BaseColor
    Normal=LinearInterpolate(A=@n22, B=@n21.Normal, Alpha=@n23) {constAlpha=0}
    CoreDarkness=ScalarParameter() {group=Lichen, parameterName=Lichen_Fuzzy_CoreDarkness}
    Power=ScalarParameter() {group=Lichen, parameterName=Lichen_Fuzzy_Power}
    EdgeBrightness=ScalarParameter() {group=Lichen, parameterName=Lichen_Fuzzy_EdgeBrightness}
   ) {materialFunction=/Engine/Functions/Engine_MaterialFunctions01/Shading/FuzzyShading}.Result
   False=@n20
  ) {group=Lichen, parameterName=UseFuzzyShading}
  n25 = ScalarParameter() {defaultValue=0.4918, group=Lichen, parameterName=Lichen_UP_DOWN_Contrast_Low}
  n26 = ScalarParameter() {defaultValue=0.776, group=Lichen, parameterName=Lichen_UP_DOWN_Contrast_High}
  n27 = ComponentMask(None=VertexNormalWS()) {b=true}
  n28 = Multiply(A=Saturate(None=LinearInterpolate(A=@n25, B=@n26, Alpha=@n27)), B=@n15)
  n29 = NamedRerouteDeclaration(None=@n28) {name=Tranisiton_Lichen, nodeColor=<rgba 1 0.8438 0 1>}
  n30 = NamedRerouteUsage()
  n31 = ScalarParameter() {defaultValue=0.744, group=Lichen, parameterName=Lichen_Transition_Contrast}
  n32 = StaticSwitchParameter(
   True=BlendMaterialAttributes(
    A=@n17
    B=SetMaterialAttributes!(MaterialAttributes=@n17, BaseColor=@n24, Normal=@n21.Normal) {attributeSetTypes [#2, #1]}
    Alpha=MaterialFunctionCall(Transition Phase=@n29, Height Texture=@n30, Contrast=@n31) {materialFunction=&HeightLerp}.Alpha
   )
   False=@n1.MaterialAttributes
  ) {group=Lichen, parameterName=UseLichen}
  FunctionOutput(None=QualitySwitch(
    Default=ShadingPathSwitch(Default=@n32, Mobile=@n1.MaterialAttributes) {desc=Migrated from FeatureLevelSwitch}
    Low=@n1.MaterialAttributes
    Medium=@n1.MaterialAttributes
   ))
  n33 = ScalarParameter() {group=Lichen, parameterName=Lichen_Hue}
  n34 = TextureObjectParameter() {group=Lichen, parameterName=Lichen_BaseColor, texture=/Game/Environments/Asteria/Foliage/Trees/FicusTree/Textures/Helios_lichen_diffuse1}
  n35 = NamedRerouteDeclaration(None=ScalarParameter() {defaultValue=261.6721, group=Lichen, parameterName=Lichen_Size_WS}) {name=LichenTextureScale, nodeColor=<rgba 0 1 0.4844 1>}
  n36 = ScalarParameter() {defaultValue=1, group=Lichen, parameterName=Lichen_PrjectionContrast}
  n37 = MaterialFunctionCall(TextureObject=@n34, TextureSize=@n35, Export Float 4=StaticBool() {value=true}, ProjectionTransitionContrast=@n36) {materialFunction=&WorldAlignedTexture}
  n38 = ComponentMask(None=@n37.XYZ Texture) {b=true, g=true, r=true}
  n39 = Multiply(
   A=VectorParameter() {group=Lichen, parameterName=LichenColor, defaultValue=<rgba 1 1 1 1>}.RGB
   B=Desaturation(
    None=MaterialFunctionCall(Hue Shift Percentage=@n33, Texture=@n38) {materialFunction=&HueShift}.Result
    Fraction=ScalarParameter() {defaultValue=0.088, group=Lichen, parameterName=Lichen_Desat}
   )
  )
  n40 = ComponentMask(None=@n38) {r=true}
  n41 = TextureObjectParameter() {
   group=Lichen
   parameterName=Lichen_Normal
   samplerType=SAMPLERTYPE_Normal
   texture=/Game/Environments/Asteria/Foliage/Trees/FicusTree/Textures/Helios_lichen_normal
  }
  n42 = MaterialFunctionCall(TextureObject=@n41, TextureSize=@n35, ProjectionTransitionContrast=@n36) {materialFunction=&WorldAlignedNormal}
  n43 = ScalarParameter() {group=Lichen, parameterName=Lichen_Normal}
  n44 = SetMaterialAttributes!(
   MaterialAttributes
   BaseColor=@n39
   Specular=LinearInterpolate(
    A=ScalarParameter() {defaultValue=0.112, group=Lichen, parameterName=Lichen_Specular_low}
    B=ScalarParameter() {defaultValue=0.5, group=Lichen, parameterName=Lichen_Specular_high}
    Alpha=@n40
   )
   Roughness=LinearInterpolate(
    A=ScalarParameter() {defaultValue=0.62, group=Lichen, parameterName=Lichen_Roughness_low}
    B=ScalarParameter() {defaultValue=0.8, group=Lichen, parameterName=Lichen_Roughness_high}
    Alpha=@n40
   )
   Normal=MaterialFunctionCall(
    BaseNormal=NamedRerouteUsage()
    AdditionalNormal=MaterialFunctionCall(Normal=@n42.XYZ Texture, Flatness=@n43) {materialFunction=&FlattenNormal}.Result
   ) {materialFunction=&BlendAngleCorrectedNormals}.Result
  ) {attributeSetTypes [
    #2
    #5
    #6
    #1
   ]}
  NamedRerouteDeclaration(None=@n44) {name=MossHighEnd, nodeColor=<rgba 1 0.5156 0 1>}
  NamedRerouteDeclaration(None=@n1.Normal) {name=OriginalNormal, nodeColor=<rgba 0 1 0.7188 1>}
  n45 = SetMaterialAttributes(
   BaseColor=@n39
   Roughness=ScalarParameter() {defaultValue=1, group=Lichen, parameterName=Lichen_Mobile_Low_Roughness}
   Normal=NamedRerouteUsage()
  ) {attributeSetTypes [
    #2
    #5
    #6
    #1
   ]}
  NamedRerouteDeclaration(None=@n45) {name=LowEndMossMaterial, nodeColor=<rgba 0 0.9297 1 1>}
  NamedRerouteDeclaration(None=ComponentMask(None=@n37.XYZ Texture) {a=true}) {name=AlphaHeightChannel, nodeColor=<rgba 0 1 0.6719 1>}
  NamedRerouteDeclaration(None=@n1.BaseColor) {name=Color OG, nodeColor=<rgba 1 0 0.1875 1>}
  Reroute(None=@n28)
  Reroute()
 ]}
