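# Material function for tree trunk materials that tints the BaseColor of incoming material attributes with two optional gradients, each behind a static switch.
# The top-down gradient multiplies in a color according to vertex normal Z, tinting upward-facing surfaces by default.
# The height gradient then tints the base and fades upward, using world Z relative to the instance origin divided by a distance parameter.
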MaterialFunction /Game/Environments/Asteria/Foliage/Trees/MasterMaterials/MaterialFunctions/MF_TreeTrunk_GradientBlends
graph {nodes [
  n1 = FunctionInput() {inputType=FunctionInput_MaterialAttributes}
  n2 = GetMaterialAttributes(None=@n1) {attributeGetTypes [#1]}
  n3 = Multiply(A=@n2.BaseColor, B=VectorParameter() {group=Gradients, parameterName=TopDown_Gradient_Color, defaultValue=<rgba 0.5312 0.2917 0.1056 1>}.RGB)
  n4 = VertexNormalWS()
  n5 = Saturate(None=LinearInterpolate(
    A=ScalarParameter() {group=Gradients, parameterName=TopDown_Gradient_Low}
    B=ScalarParameter() {defaultValue=1, group=Gradients, parameterName=TopDown_Gradient_High}
    Alpha=ComponentMask(None=@n4) {b=true}
   ))
  n6 = StaticSwitchParameter(
   True=SetMaterialAttributes!(MaterialAttributes=@n2.MaterialAttributes, BaseColor=LinearInterpolate(A=@n2.BaseColor, B=@n3, Alpha=@n5)) {attributeSetTypes [#1]}
   False=@n1
  ) {group=Gradients, parameterName=UseTrunkBlendGradients}
  n7 = GetMaterialAttributes(None=@n6) {attributeGetTypes [#1]}
  n8 = ScalarParameter() {defaultValue=-0.808, group=Gradients, parameterName=Height_Gradient_Low}
  n9 = ScalarParameter() {defaultValue=1.1037, group=Gradients, parameterName=Height_Gradient_High}
  n10 = TransformPosition(None=Constant3Vector()) {transformSourceType=TRANSFORMPOSSOURCE_Instance, transformType=TRANSFORMPOSSOURCE_World}
  n11 = Divide(
   A=Subtract(A=WorldPosition().Z, B=ComponentMask(None=@n10) {b=true})
   B=ScalarParameter() {defaultValue=1500, group=Gradients, parameterName=Height_Gradient_Distance}
  )
  n12 = LinearInterpolate(
   A=Multiply(A=@n7.BaseColor, B=VectorParameter() {group=Gradients, parameterName=Height_Gradient_Color, defaultValue=<rgba 0.5312 0.2917 0.1056 1>}.RGB)
   B=@n7.BaseColor
   Alpha=Saturate(None=LinearInterpolate(A=@n8, B=@n9, Alpha=@n11))
  )
  FunctionOutput(None=StaticSwitchParameter(
    True=SetMaterialAttributes(MaterialAttributes=@n7.MaterialAttributes, BaseColor=@n12) {attributeSetTypes [#1]}
    False=@n6
   ) {group=Gradients, parameterName=UseTrunkHeightBlendGradients})
  PixelNormalWS()
 ]}
