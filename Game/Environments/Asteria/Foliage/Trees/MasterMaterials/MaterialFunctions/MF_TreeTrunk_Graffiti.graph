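# Material function that overlays painted graffiti on tree trunk material attributes, with static bypasses for impostor bakes and disabled graffiti.
# Graffiti comes from a texture or array, with optional UV remapping, position-seeded random variation, and hue shifts using an HSV-to-RGB LUT or built-in conversion.
# Texture alpha, height gradients, optional UV masking and normal-based wear form the mask that blends base color, specular and roughness, leaving normals unchanged.
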MaterialFunction /Game/Environments/Asteria/Foliage/Trees/MasterMaterials/MaterialFunctions/MF_TreeTrunk_Graffiti
graph {nodes [
  n1 = GetMaterialAttributes(None=NamedRerouteDeclaration(None=FunctionInput() {inputType=FunctionInput_MaterialAttributes}) {name=inMA, nodeColor=<rgba 1 0.7734 0 1>}) {attributeGetTypes [
    #1
    #2
    #3
    #4
   ]}
  n2 = NamedRerouteUsage()
  n3 = ComponentMask(None=@n2) {b=true, g=true, r=true}
  n4 = ComponentMask(None=@n2) {a=true}
  n5 = Multiply(
   A=Multiply(A=NamedRerouteDeclaration(None=@n4) {name=gTexAlpha, nodeColor=<rgba 0.9219 1 0 1>}, B=NamedRerouteUsage())
   B=ScalarParameter() {defaultValue=1, group=Graffiti, parameterName=Graffiti Opacity, sortPriority=1}
  )
  n6 = StaticSwitchParameter(True=Multiply(A=@n5, B=NamedRerouteUsage()), False=@n5) {
   defaultValue=true
   desc=Masks remapped UV's so that trunk is shown past Min/Max UV
   group=Graffiti
   parameterName=Graffiti MaskUVs
   sortPriority=255
  }
  n7 = StaticSwitchParameter(True=@n6, False=@n5) {defaultValue=true, desc=Remaps UV's to a specific portion of the UV space, group=Graffiti, parameterName=Graffiti RemapUVs, sortPriority=255}
  n8 = SetMaterialAttributes!(
   MaterialAttributes=@n1.MaterialAttributes
   BaseColor=LinearInterpolate(
    A=@n1.BaseColor
    B=NamedRerouteDeclaration(None=@n3) {name=texRGB, nodeColor=<rgba 0 1 0.1562 1>}
    Alpha=NamedRerouteDeclaration(None=@n7) {name=gAlpha, nodeColor=<rgba 0.5234 1 0 1>}
   )
   Specular=LinearInterpolate(
    A=@n1.Specular
    B=ScalarParameter() {defaultValue=0.5, group=Graffiti, parameterName=Graffiti Specular, sortPriority=1}
    Alpha=NamedRerouteUsage()
   )
   Roughness=LinearInterpolate(
    A=@n1.Roughness
    B=ScalarParameter() {defaultValue=0.5, group=Graffiti, parameterName=Graffiti Roughness, sortPriority=1}
    Alpha=NamedRerouteUsage()
   )
   Normal=NamedRerouteDeclaration(None=@n1.Normal) {name=normalIn, nodeColor=<rgba 0.7109 1 0 1>}
  ) {attributeSetTypes [
    #1
    #2
    #3
    #4
   ]}
  n9 = NamedRerouteUsage()
  n10 = StaticBool()
  FunctionOutput(None=StaticSwitch(
    True=NamedRerouteUsage()
    False=StaticSwitchParameter(True=@n8, False=@n9) {group=Graffiti, parameterName=UseGraffiti, sortPriority=255}
    Value=FunctionInput(Preview=@n10) {bUsePreviewValueAsDefault=true, inputName=IsImpostorBake, inputType=FunctionInput_StaticBool, sortPriority=1}
   ))
  n11 = TransformPosition(None=Constant3Vector()) {transformSourceType=TRANSFORMPOSSOURCE_Instance, transformType=TRANSFORMPOSSOURCE_World}
  n12 = ComponentMask(None=@n11) {b=true}
  n13 = Subtract(A=Subtract(A=WorldPosition().Z, B=@n12), B=NamedRerouteUsage())
  n14 = Divide(A=@n13, B=ScalarParameter() {defaultValue=1500, group=Graffiti, parameterName=Graffiti Height, sortPriority=20})
  n15 = ScalarParameter() {defaultValue=1, group=Graffiti, parameterName=Graffiti HeightContrast, sortPriority=20}
  n16 = OneMinus(None=Saturate(None=Power(Base=@n14, Exp=@n15)))
  n17 = Divide(A=@n13, B=ScalarParameter() {defaultValue=250, group=Graffiti, parameterName=Graffiti BottomHeight, sortPriority=20})
  n18 = Divide(B=ScalarParameter() {defaultValue=1, group=Graffiti, parameterName=Graffiti BottomContrast, sortPriority=20})
  n19 = NamedRerouteDeclaration(None=Saturate(None=Power(Base=@n17, Exp=@n18))) {name=bottomGradient, nodeColor=<rgba 0 1 0.6484 1>}
  n20 = NamedRerouteDeclaration(None=Saturate(None=Multiply(A=@n16, B=@n19))) {name=heightGradient, nodeColor=<rgba 1 0 0.7031 1>}
  n21 = PixelNormalWS()
  n22 = NamedRerouteUsage()
  n23 = Divide(B=OneMinus(None=DotProduct(A=@n21, B=@n22)))
  n24 = ScalarParameter() {
   defaultValue=1
   desc=Set to Zero to only show graffiti paint based on texture alpha.
   group=Graffiti
   parameterName=Graffiti PaintWearIntensity
   sortPriority=21
  }
  n25 = NamedRerouteUsage()
  n26 = ScalarParameter() {defaultValue=7, group=Graffiti, parameterName=Graffiti PaintWearFadeMax, sortPriority=21}
  n27 = PixelNormalWS()
  n28 = Constant3Vector() {constant=<rgba -0.2 0.5 -1 0>}
  n29 = OneMinus(None=NamedRerouteUsage())
  n30 = LinearInterpolate(
   A=Saturate(None=@n24)
   B=Add(
    A=Multiply(A=@n25, B=@n26) {constB=7}
    B=ScalarParameter() {defaultValue=3, group=Graffiti, parameterName=Graffiti PaintWearFadeMin, sortPriority=21}
   ) {constB=3}
   Alpha=Multiply(A=DotProduct(A=@n27, B=@n28), B=Power(Base=@n29))
  ) {constB=10}
  n31 = Saturate(None=Power(
    Base=Divide(A=@n23, B=@n30)
    Exp=ScalarParameter() {defaultValue=1, group=Graffiti, parameterName=Graffiti PaintWearContrast, sortPriority=20}
   ))
  n32 = StaticSwitchParameter(True=Multiply(A=@n20, B=NamedRerouteDeclaration(None=@n31) {name=wearGradient, nodeColor=<rgba 1 0 0.5625 1>}), False=@n20) {defaultValue=true, desc=Uses normal map to wear out paint., group=Graffiti, parameterName=Graffiti UsePaintWear, sortPriority=255}
  NamedRerouteDeclaration(None=@n32) {name=trunkGradients, nodeColor=<rgba 0 0.8828 1 1>}
  n33 = ObjectPositionWS()
  n34 = ScalarParameter() {
   defaultValue=1
   desc=How slow the randomization happens based on World Position. Lower values change faster.
   group=Graffiti
   parameterName=Graffiti RandSmooth
   sortPriority=64
  }
  n35 = Divide(
   A=MaterialFunctionCall(f3=ComponentMask(None=@n33) {b=true, g=true, r=true}) {materialFunction=&AddComponents}.f3
   B=Multiply(B=@n34) {constA=1000}
  )
  NamedRerouteDeclaration(None=Multiply(A=Frac(None=@n35), B=NamedRerouteUsage())) {name=randFloat_IPSmooth, nodeColor=<rgba 0.4766 1 0 1>}
  n36 = ScalarParameter() {bUseCustomPrimitiveData=true, defaultValue=-0.03, group=Color, parameterName=Hue Variation_primdata, primitiveDataIndex=5}
  n37 = StaticSwitchParameter(True=@n36, False=ScalarParameter() {group=Graffiti, parameterName=Graffiti HueShift, sortPriority=11}) {
   desc=Uses HueShiftBlueVertexColor as per-instance hue-shift value
   group=Graffiti
   parameterName=Graffiti PerInstanceHueShift
   sortPriority=255
  }
  n38 = NamedRerouteUsage()
  NamedRerouteDeclaration(None=StaticSwitchParameter(True=Add(A=@n37, B=@n38), False=@n37) {desc=Random Hue-shift, group=Graffiti, parameterName=Graffiti RandomHue, sortPriority=255}) {name=hueShift, nodeColor=<rgba 1 0 0.5391 1>}
  n39 = Ceil(None=Subtract(A=NamedRerouteUsage(), B=ScalarParameter() {group=Graffiti, parameterName=Graffiti Min U, sortPriority=10}))
  n40 = Ceil(None=Subtract(A=ScalarParameter() {defaultValue=1, group=Graffiti, parameterName=Graffiti Max U, sortPriority=10}, B=NamedRerouteUsage()))
  NamedRerouteDeclaration(None=Saturate(None=Multiply(A=@n39, B=@n40))) {name=UMask, nodeColor=<rgba 1 0.2109 0 1>}
  n41 = NamedRerouteUsage()
  n42 = ScalarParameter() {group=Graffiti, parameterName=Graffiti HeightOffset Min, sortPriority=18}
  n43 = ScalarParameter() {defaultValue=50, group=Graffiti, parameterName=Graffiti HeightOffset Max, sortPriority=19}
  NamedRerouteDeclaration(None=StaticSwitchParameter(
    True=MaterialFunctionCall(Input=@n41, Target Low=@n42, Target High=@n43) {materialFunction=&RemapValueRange}.Result
    False=ScalarParameter() {group=Graffiti, parameterName=Graffiti HeightOffset, sortPriority=20}
   ) {group=Graffiti, parameterName=Graffiti RandomMapping, sortPriority=255}) {name=hOffset, nodeColor=<rgba 0 0.1562 1 1>}
  NamedRerouteDeclaration(None=MaterialFunctionCall!("UV (2D Seed)"=ComponentMask(None=@n11) {g=true, r=true}, Additional Seed=@n12) {materialFunction=&/MF_SimpleRandom}.Random Float) {name=randFloat_IP, nodeColor=<rgba 0.4766 1 0 1>}
  n44 = TextureCoordinate()
  n45 = ComponentMask(None=@n44) {r=true}
  n46 = StaticSwitchParameter(True=@n45, False=Multiply(A=ScalarParameter() {defaultValue=1, group=Graffiti, parameterName=Graffiti Tile U, sortPriority=10}, B=@n45)) {defaultValue=true, desc=Remaps UV's to a specific portion of the UV space, group=Graffiti, parameterName=Graffiti RemapUVs, sortPriority=255}
  n47 = StaticSwitchParameter(
   True=MaterialFunctionCall(
    Input=NamedRerouteDeclaration(None=@n46) {name=UCoord, nodeColor=<rgba 0 1 0.5313 1>}
    Input Low=ScalarParameter() {group=Graffiti, parameterName=Graffiti Min U, sortPriority=10}
    Input High=ScalarParameter() {defaultValue=1, group=Graffiti, parameterName=Graffiti Max U, sortPriority=10}
   ) {materialFunction=&RemapValueRange}.Result
   False=NamedRerouteUsage()
  ) {defaultValue=true, desc=Remaps UV's to a specific portion of the UV space, group=Graffiti, parameterName=Graffiti RemapUVs, sortPriority=255}
  n48 = Multiply(A=ComponentMask(None=@n44) {g=true}, B=ScalarParameter() {defaultValue=1, group=Graffiti, parameterName=Graffiti Tile V, sortPriority=10})
  n49 = StaticSwitchParameter(
   True=MaterialFunctionCall(
    Input=NamedRerouteDeclaration(None=@n48) {name=VCoord, nodeColor=<rgba 1 0.0234 0 1>}
    Input Low=ScalarParameter() {group=Graffiti, parameterName=Graffiti Min V, sortPriority=10}
    Input High=ScalarParameter() {defaultValue=1, group=Graffiti, parameterName=Graffiti Max V, sortPriority=10}
   ) {materialFunction=&RemapValueRange}.Result
   False=NamedRerouteUsage()
  ) {defaultValue=true, desc=Remaps UV's to a specific portion of the UV space, group=Graffiti, parameterName=Graffiti RemapUVs, sortPriority=255}
  n50 = StaticSwitchParameter(
   True=MaterialFunctionCall(
    Input=NamedRerouteUsage()
    Target Low=ScalarParameter() {group=Graffiti, parameterName=Graffiti Offset V Min, sortPriority=10}
    Target High=ScalarParameter() {defaultValue=1, group=Graffiti, parameterName=Graffiti Offset V Max, sortPriority=10}
   ) {materialFunction=&RemapValueRange}.Result
   False=ScalarParameter() {group=Graffiti, parameterName=Graffiti Offset V, sortPriority=10}
  ) {group=Graffiti, parameterName=Graffiti RandomMapping, sortPriority=255}
  n51 = AppendVector(A=@n47, B=Add(A=@n49, B=NamedRerouteDeclaration(None=@n50) {name=vOffset, nodeColor=<rgba 0 0.1094 1 1>}))
  n52 = Round(None=Multiply(A=ScalarParameter() {group=Graffiti, parameterName=Graffiti MaxTextures, sortPriority=11}, B=NamedRerouteUsage()))
  n53 = NamedRerouteDeclaration(None=StaticSwitchParameter(True=AppendVector(A=@n51, B=@n52), False=@n51) {desc=Uses Texture2DArray to randomize textures, group=Graffiti, parameterName=Graffiti RandomTexture, sortPriority=255}) {name=UVs, nodeColor=<rgba 0.4766 0 1 1>}
  n54 = TextureSampleParameter2DArray!(UVs=@n53, Apply View MipBias) {
   group=Graffiti
   parameterName=Graffiti TextureArray
   samplerSource=SSM_Wrap_WorldGroupSettings
   texture=/Game/Environments/Asteria/Foliage/Trees/MasterMaterials/DefaultTextures/Graffiti/TA_GraffitiArray
  }
  n55 = TextureSampleParameter2D(UVs=@n53) {
   group=Graffiti
   parameterName=Graffiti Texture
   samplerSource=SSM_Wrap_WorldGroupSettings
   texture=/Game/Environments/Asteria/Foliage/Trees/MasterMaterials/DefaultTextures/Graffiti/T_graffiti_001
  }
  n56 = MaterialFunctionCall!(Float4=StaticSwitchParameter(True=@n54.RGBA, False=@n55.RGBA) {group=Graffiti, parameterName=Graffiti RandomTexture, sortPriority=255}) {materialFunction=/Game/Global/MaterialFunctions/MF_BreakComponents}
  n57 = RgbToHsv!(None=@n56.RGB)
  n58 = Add(A=ComponentMask(None=NamedRerouteDeclaration(None=@n57) {name=hsvSample, nodeColor=<rgba 0.3828 0 1 1>}) {r=true}, B=NamedRerouteUsage())
  n59 = ComponentMask(None=@n57) {g=true}
  n60 = TextureObject() {
   samplerType=SAMPLERTYPE_LinearColor
   texture=/Game/Environments/Asteria/Foliage/Trees/MasterMaterials/DefaultTextures/Graffiti/T_HSVLUT_float
  }
  n61 = TextureObject() {samplerType=SAMPLERTYPE_LinearColor, texture=/Game/Environments/Asteria/Foliage/Trees/MasterMaterials/DefaultTextures/Graffiti/T_HSVLUT_16b}
  n62 = TextureSample(UVs=AppendVector(A=@n58, B=OneMinus(None=@n59)), Tex=ShadingPathSwitch(Default=@n60, Deferred=@n60, Forward=@n61, Mobile=@n61)) {samplerType=SAMPLERTYPE_LinearColor, texture=&T_HSVLUT_16b}
  n63 = NamedRerouteUsage()
  n64 = ComponentMask(None=NamedRerouteUsage()) {b=true, g=true}
  n65 = StaticSwitchParameter(True=Multiply(A=@n62.RGB, B=ComponentMask(None=@n63) {b=true}), False=HsvToRgb!(None=AppendVector(A=@n58, B=@n64))) {
   defaultValue=true
   desc=Uses a LUT for RGB to HSV conversion in place of the built-in function
   group=Graffiti
   parameterName=Graffiti UseHSV LUT
   sortPriority=255
  }
  NamedRerouteDeclaration(None=AppendVector(A=@n65, B=@n56.A)) {name=rgbaSample, nodeColor=<rgba 0 1 0.9531 1>}
  Reroute(None=VertexInterpolator(VS=VertexTangentWS()).PS)
 ]}
