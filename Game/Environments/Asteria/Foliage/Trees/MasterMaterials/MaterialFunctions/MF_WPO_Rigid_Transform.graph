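# Material function for Asteria trees that builds world position offset for wind from baked skeletal animation textures.
# Vertices rotate rigidly around bind-pose pivots by blended rest and wind quaternions, plus texture position offsets, with UV1 indexing the timeline.
# A storm-blended, height-masked, length-preserving sine bend covers the whole tree and tree add-ons.
# Mesh and custom normals follow that rotation, while Normal and Quaternion outputs use a separate baked-transform texture.
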MaterialFunction /Game/Environments/Asteria/Foliage/Trees/MasterMaterials/MaterialFunctions/MF_WPO_Rigid_Transform
graph {nodes [
  n1 = NamedRerouteUsage()
  n2 = MaterialFunctionCall(Float3=@n1) {materialFunction=&BreakOutFloat3Components}
  n3 = Divide(A=Add(A=@n2.R, B=@n2.R)) {constB=10}
  n4 = Multiply(A=VertexColor().A) {constB=33.33}
  n5 = Multiply(A=Add(A=Frac(None=@n3), B=Frac(None=@n4))) {constB=23.33}
  Abs(None=@n5)
  n6 = Time()
  n7 = NamedRerouteUsage()
  n8 = Add(A=Multiply(A=@n6, B=ScalarParameter() {defaultValue=0.25, group=Animation WInd, parameterName=Wind_Bend_Speed_Default}), B=@n7)
  n9 = Add(A=Multiply(A=@n6, B=ScalarParameter() {defaultValue=1, group=Animation WInd, parameterName=Wind_Bend_Speed_Storm}), B=@n7)
  n10 = NamedRerouteUsage()
  n11 = LinearInterpolate(A=Sine(None=@n8), B=Sine(None=@n9), Alpha=@n10)
  n12 = LinearInterpolate(A=Cosine(None=Multiply(A=@n8) {constB=1.13}), B=Cosine(None=Multiply(A=@n9) {constB=1.13}), Alpha=@n10)
  n13 = Multiply(
   A=MaterialFunctionCall(X=Add(A=@n11) {constB=0.75}, Y=Multiply(A=@n12) {constB=0.618}) {materialFunction=&MakeFloat3}.Result
   B=LinearInterpolate(
    A=ScalarParameter() {defaultValue=16, group=Animation WInd, parameterName=Wind_Bend_Strength}
    B=ScalarParameter() {defaultValue=80, group=Animation WInd, parameterName=Wind_Bend_Strength_Storm}
    Alpha=NamedRerouteUsage()
   )
  )
  n14 = Add(A=@n1, B=Constant3Vector() {constant=<rgba 17 53 101 0>})
  Add(
   A=@n13
   B=Multiply(
    A=MaterialFunctionCall!(Seed Position=@n14) {materialFunction=/Game/Environments/Asteria/Foliage/Wind/MF_Rand3D_to_1D}.Rand 1D
    B=ScalarParameter() {group=Animation WInd, parameterName=detailNoise}
   )
  )
  Constant3Vector()
  Constant() {r=1}
  Constant()
  Constant()
  n15 = NamedRerouteUsage()
  n16 = MaterialFunctionCall() {materialFunction=&LocalPosition}
  n17 = TextureSampleParameter2D(UVs=NamedRerouteUsage()) {
   group=Animation
   parameterName=Bind Pose 1D
   samplerSource=SSM_Wrap_WorldGroupSettings
   samplerType=SAMPLERTYPE_LinearColor
   texture=/Game/Environments/Asteria/Foliage/Trees/MasterMaterials/DefaultTextures/TreeLarge8PrunedInterior_SKEL_ANIM_BIND_V01
  }
  n18 = TextureSampleParameter2D!(UVs=NamedRerouteUsage(), Level, Apply View MipBias) {
   group=Animation
   mipValueMode=TMVM_MipLevel
   parameterName=Rotation Offset
   samplerType=SAMPLERTYPE_LinearColor
   texture=/Game/Environments/Asteria/Foliage/Trees/MasterMaterials/DefaultTextures/TREE_SKEL_ANIM_ORIENT_V01
  }
  n19 = LinearInterpolate(
   A=MaterialFunctionCall(A=Constant() {r=-1}) {materialFunction=&MakeFloat4}.Result
   B=@n18.RGBA
   Alpha=NamedRerouteUsage()
  )
  n20 = TextureSampleParameter2D!(UVs=NamedRerouteUsage(), Apply View MipBias) {
   group=Animation
   parameterName=WindRotation
   samplerType=SAMPLERTYPE_LinearColor
   texture=/Game/Environments/Asteria/Foliage/Trees/MasterMaterials/DefaultTextures/TreeLarge8_old_SKEL_ANIM_WIND_ORIENT_V01
  }
  n21 = Normalize(VectorInput=LinearInterpolate(A=@n19, B=@n20.RGBA, Alpha=NamedRerouteUsage()))
  n22 = StaticBool()
  n23 = MaterialFunctionCall!(Vector=Subtract(A=@n16."Instance Local Position (Excluding Offsets)", B=@n17.RGB), Quat=@n21, Invert Quaternion=@n22) {materialFunction=/Engine/Functions/Engine_MaterialFunctions03/Math/MultiplyVectorWithQuaternion}
  n24 = TextureSampleParameter2D(UVs=NamedRerouteUsage()) {
   group=Animation
   parameterName=OffsetPositionTex
   samplerSource=SSM_Wrap_WorldGroupSettings
   samplerType=SAMPLERTYPE_LinearColor
   texture=/Game/Environments/Asteria/Foliage/Trees/MasterMaterials/DefaultTextures/TreeLarge8PrunedInterior_SKEL_ANIM_POSITION_V01
  }
  n25 = Multiply(A=@n24.RGB, B=NamedRerouteUsage())
  n26 = TextureSampleParameter2D(UVs=NamedRerouteUsage()) {
   group=Animation
   parameterName=WindOffset
   samplerSource=SSM_Wrap_WorldGroupSettings
   samplerType=SAMPLERTYPE_LinearColor
   texture=/Game/Environments/Asteria/Foliage/Trees/MasterMaterials/DefaultTextures/TreeLarge8_old_SKEL_ANIM_WIND_POSITION_V01
  }
  n27 = NamedRerouteUsage()
  n28 = Subtract(A=Add(A=Add(A=@n23.Result, B=@n17.RGB), B=LinearInterpolate(A=@n25, B=@n26.RGB, Alpha=@n27)), B=@n16.Instance Local Position)
  n29 = Transform(None=@n28) {transformSourceType=TRANSFORMSOURCE_Local}
  FunctionOutput(None=StaticSwitch(True=@n15, False=Add(A=@n15, B=@n29), Value=NamedRerouteUsage())) {outputName=Offset}
  n30 = Transform(None=VertexNormalWS()) {transformSourceType=TRANSFORMSOURCE_World, transformType=TRANSFORM_Local}
  n31 = StaticSwitchParameter(True=Constant() {r=0.3333}, False=Constant() {r=0.5}) {parameterName=ConstainsScaleInfo}
  n32 = Multiply(
   A=ScalarParameter() {defaultValue=10, group=BakedTransforms, parameterName=numSockets, sortPriority=5}
   B=StaticSwitchParameter(True=Constant() {r=3}, False=Constant() {r=2}) {desc=Whether the baked transformation texture includes scale information, parameterName=ConstainsScaleInfo}
  ) {constB=3}
  n33 = Add(
   A=Multiply(A=VertexColor().R, B=@n31) {constB=0.5}
   B=Multiply(A=Divide(B=@n32) {constB=23}, B=ScalarParameter() {defaultValue=0.5, group=BakedTransforms, parameterName=SubPixelMult}) {constB=0.25}
  ) {constB=-0.1}
  n34 = Add(A=@n33, B=@n31) {constB=0.5}
  n35 = ScalarParameter() {defaultValue=2, group=Deformation, parameterName=PreviewFrame}
  n36 = ScalarParameter() {defaultValue=40, group=BakedTransforms, parameterName=numFrames, sortPriority=10}
  n37 = Divide(B=@n36) {constB=40}
  n38 = Add(A=Time(), B=NamedRerouteUsage())
  n39 = Multiply(A=VertexColor().R, B=ScalarParameter() {defaultValue=0.3333, group=Animation, parameterName=PerLeafAnimationRandomOffsetScale}) {constB=0.1}
  n40 = FunctionInput(Preview=Subtract(A=Subtract(A=@n38, B=@n39))) {bUsePreviewValueAsDefault=true, inputName=TimeInput, inputType=FunctionInput_Scalar}
  n41 = Multiply(
   A=Add(A=Time(), B=ScalarParameter() {parameterName=TimeOffset})
   B=ScalarParameter() {defaultValue=1, group=Deformation, parameterName=SpeedMultiplier, sortPriority=20}
  )
  n42 = Divide(B=@n36) {constA=30}
  n43 = DynamicParameter()
  n44 = ShaderStageSwitch(PixelShader=VertexInterpolator(VS=@n43.Param1).PS, VertexShader=@n43.Param1)
  n45 = PerInstanceCustomData!(DefaultValue)
  n46 = StaticSwitchParameter(True=ShaderStageSwitch(PixelShader=VertexInterpolator(VS=@n45).PS, VertexShader=@n45), False=PerInstanceRandom()) {parameterName=PerInstanceCustomData}
  n47 = Frac(None=Add(A=Multiply(A=@n41, B=@n42), B=StaticSwitchParameter(True=@n44, False=@n46) {parameterName=DynamicParameter}) {constB=0})
  n48 = StaticSwitchParameter(
   True=Add(A=Multiply(A=@n35, B=@n37) {constB=1.5}, B=Multiply(A=@n37) {constB=0.5})
   False=Max(A=StaticSwitchParameter(True=@n40, False=@n47) {parameterName=InputTime}, B=Multiply(A=@n37) {constB=1.5})
  ) {group=Deformation, parameterName=FrameOverride}
  n49 = TextureSampleParameter2D(UVs=MaterialFunctionCall(X=@n48, Y=@n34) {materialFunction=&MakeFloat2}.Result) {
   desc=position
   group=BakedTransforms
   parameterName=AnimationTexture
   samplerType=SAMPLERTYPE_LinearColor
   sortPriority=0
   texture=/Game/Global/Textures/Default/Blanks/T_BlackVectorDisplacement
  }
  n50 = LinearInterpolate(A=Constant4Vector() {constant=<rgba 0 0 0 1>}.RGBA, B=@n49.RGBA, Alpha=ScalarParameter() {defaultValue=1, parameterName=QuaternionMix})
  n51 = VertexInterpolator(VS=@n50)
  FunctionOutput(None=Transform(None=MaterialFunctionCall(Vector=@n30, Quat=@n51.PS) {materialFunction=&MultiplyVectorWithQuaternion}.Result) {transformSourceType=TRANSFORMSOURCE_Local}) {outputName=Normal}
  FunctionOutput(None=@n51.PS) {outputName=Quaternion}
  n52 = FunctionInput(Preview=Constant3Vector() {constant=<rgba 0 0 1 0>}) {
   bUsePreviewValueAsDefault=true
   description=Takes a tengent space normal and rotates it by the WPO function
   inputName=TangentNormalToTransform
  }
  n53 = Transform(None=@n52) {transformType=TRANSFORM_Local}
  FunctionOutput(None=Transform(None=MaterialFunctionCall(Vector=@n53, Quat=@n21, Invert Quaternion=@n22) {materialFunction=&MultiplyVectorWithQuaternion}.Result) {transformSourceType=TRANSFORMSOURCE_Local, transformType=TRANSFORM_Tangent}) {outputName=Custom Tangent Normal}
  n54 = VertexInterpolator(VS=Custom!(None) {code="return Parameters.PreSkinnedNormal;", description=Get Preskinned Normal})
  FunctionOutput(None=Transform(None=MaterialFunctionCall(Vector=@n54.PS, Quat=@n21, Invert Quaternion=@n22) {materialFunction=&MultiplyVectorWithQuaternion}.Result) {transformSourceType=TRANSFORMSOURCE_Local, transformType=TRANSFORM_Tangent}) {description=This returns the unmodified mesh normals transformed to tangent space with the correct WPO applied, outputName=Mesh Normals}
  n55 = MaterialFunctionCall(Float4=@n18.RGBA) {
   desc=This is will remove that Array Texture bug wiht alpha channel
   materialFunction=&BreakOutFloat4Components
  }
  MaterialFunctionCall(X=@n55.R, Y=@n55.R, Z=@n55.R, A=Constant() {r=0.5}) {materialFunction=&MakeFloat4}
  Multiply(A=@n28, B=ScalarParameter() {defaultValue=1.3141, parameterName=WPO_Gain})
  NamedRerouteDeclaration(None=MaterialFunctionCall!(Float In=ScalarParameter() {defaultValue=1, group=Animation, parameterName=rotGain}, Vector_In, Index=Constant() {r=30}) {materialFunction=&/MF_AsteriaTree_ArrayAttribute}.Float Out) {name=RotationGain, nodeColor=<rgba 1 0.4453 0 1>}
  n56 = MaterialFunctionCall(Float In=ScalarParameter() {group=Animation, parameterName=Wind_Default, sliderMax=1}, Index=Constant() {r=31}) {materialFunction=&MF_AsteriaTree_ArrayAttribute}
  NamedRerouteDeclaration(None=@n56.Float Out) {name=Wind_Default, nodeColor=<rgba 1 0 0.7969 1>}
  n57 = ScalarParameter() {defaultValue=0.15, group=Animation WInd, parameterName=AnimationSpeed_Default}
  n58 = ScalarParameter() {defaultValue=1.1982, group=Animation WInd, parameterName=AnimationSpeed_Storm}
  n59 = Floor(None=NamedRerouteUsage())
  n60 = OneMinus(None=ComponentMask(None=TextureCoordinate() {coordinateIndex=1}) {g=true})
  n61 = AppendVector(A=Multiply(A=@n40, B=LinearInterpolate(A=@n57, B=@n58, Alpha=@n59)) {constB=0.1}, B=@n60)
  NamedRerouteDeclaration(None=@n61) {name=AnimationTimeLine, nodeColor=<rgba 0.2656 1 0 1>}
  n62 = AppendVector(A=ScalarParameter() {defaultValue=0.5, group=Animation, parameterName=Rotation}, B=@n60)
  NamedRerouteDeclaration(None=@n62) {name=AnimationTimeline 2, nodeColor=<rgba 0.4531 0 1 1>}
  n63 = Constant3Vector()
  NamedRerouteDeclaration(None=StaticSwitch(
    True=ActorPositionWS()
    False=TransformPosition(None=@n63) {transformSourceType=TRANSFORMPOSSOURCE_Instance, transformType=TRANSFORMPOSSOURCE_World}
    Value=NamedRerouteUsage()
   )) {name=InstancePosition, nodeColor=<rgba 0 1 0.5781 1>}
  n64 = NamedRerouteUsage()
  n65 = Subtract(A=NamedRerouteUsage(), B=@n64)
  n66 = WorldPosition()
  n67 = Constant3Vector()
  n68 = StaticSwitch(
   True=Add(A=@n66.XYZ, B=FunctionInput(Preview=@n67) {bUsePreviewValueAsDefault=true, inputName=OptionalWPO, sortPriority=798})
   False=@n66.XYZ
   Value=NamedRerouteUsage()
  )
  n69 = Subtract(
   A=ComponentMask(None=NamedRerouteDeclaration(None=@n68) {name=WPO_plus_Offset, nodeColor=<rgba 0 1 0.8594 1>}) {b=true}
   B=ComponentMask(None=NamedRerouteUsage()) {b=true}
  )
  n70 = ScalarParameter() {group=Animation, parameterName=TrunkBendOffset}
  n71 = ObjectRadius()
  n72 = Saturate(None=Divide(A=Subtract(A=@n69, B=@n70), B=Multiply(A=@n71) {constB=2}))
  n73 = Subtract(A=NamedRerouteUsage(), B=Add(A=@n64, B=Multiply(A=@n13, B=@n72)))
  n74 = Divide(
   A=MaterialFunctionCall(Vector 3=@n73) {materialFunction=/Engine/Functions/Engine_MaterialFunctions02/Utility/VectorLength}.V3 Length
   B=MaterialFunctionCall(Vector 3=@n65) {materialFunction=&VectorLength}.V3 Length
  )
  NamedRerouteDeclaration(None=Subtract(A=@n65, B=Divide(A=@n73, B=@n74))) {name=Wind, nodeColor=<rgba 1 0.5156 0 1>}
  NamedRerouteDeclaration(None=MaterialFunctionCall!(
    Position=NamedRerouteUsage()
    IsTrees=StaticBool() {value=true}
    IsGrass
    IsMediumPlant
    IsMediumPlant_2024Anim
    IsMediumPlant2ndLevel
    MediumPlant_BasePivotPostion
   ) {materialFunction=/Game/Environments/Asteria/Foliage/Wind/MF_WindActorData}.StormMask) {name=StormMask, nodeColor=<rgba 0.7344 1 0 1>}
  NamedRerouteDeclaration(None=@n5) {name=RandomNumber, nodeColor=<rgba 0 1 0.2734 1>}
  n75 = FunctionInput(Preview=StaticBool()) {bUsePreviewValueAsDefault=true, inputName=TreeAddOnOverallBend, inputType=FunctionInput_StaticBool, sortPriority=798}
  NamedRerouteDeclaration(None=@n75) {name=IsTreeAddon, nodeColor=<rgba 0.4531 0 1 1>}
  Normalize(VectorInput=@n19)
  PerInstanceCustomData3Vector!(DefaultValue)
  n76 = NamedRerouteUsage()
  n77 = WorldPosition()
  n78 = Divide(
   A=Multiply(A=Subtract(A=@n76, B=@n77.XYZ), B=VectorParameter() {defaultValue=<rgba 1.5 1.5 1 1>}.RGB)
   B=Divide(A=MaterialFunctionCall() {materialFunction=&ObjectLocalBounds}.Local Bounds Max) {constB=1}
  )
  Power(
   Base=Saturate(None=MaterialFunctionCall(Vector 3=@n78) {materialFunction=&VectorLength}.V3 Length)
   Exp=ScalarParameter() {defaultValue=1, group=Animation WInd, parameterName=BendMaskPow}
  ) {constExponent=2.71}
  Reroute(None=@n56.Float Out)
  Reroute(None=@n61)
  Reroute(None=@n62)
  Reroute(None=@n75)
  Reroute(None=@n68)
  n79 = TextureSampleParameter2D(UVs=MaterialFunctionCall(X=@n48, Y=@n33) {materialFunction=&MakeFloat2}.Result) {
   desc=position
   group=BakedTransforms
   parameterName=AnimationTexture
   samplerType=SAMPLERTYPE_LinearColor
   sortPriority=0
   texture=&T_BlackVectorDisplacement
  }
  n80 = TextureSampleParameter2D(UVs=MaterialFunctionCall(X=@n48, Y=Add(A=@n34, B=@n31)) {materialFunction=&MakeFloat2}.Result) {
   desc=position
   group=BakedTransforms
   parameterName=AnimationTexture
   samplerType=SAMPLERTYPE_LinearColor
   sortPriority=0
   texture=&T_BlackVectorDisplacement
  }
  n81 = PreSkinnedPosition()
  n82 = TextureSampleParameter2D(UVs=MaterialFunctionCall(X=Multiply(A=@n37) {constB=0.5}, Y=@n33) {materialFunction=&MakeFloat2}.Result) {
   desc=position
   group=BakedTransforms
   parameterName=AnimationTexture
   samplerType=SAMPLERTYPE_LinearColor
   sortPriority=0
   texture=&T_BlackVectorDisplacement
  }
  n83 = Subtract(A=@n81, B=@n82.RGB) {constB=0}
  n84 = StaticSwitchParameter(
   True=StaticSwitchParameter(True=Multiply(A=@n83, B=@n80.RGB) {constB=0.5}, False=@n83) {group=Deformation, parameterName=ApplyScaleAnimation, sortPriority=10}
   False=@n83
  ) {group=BakedTransforms, parameterName=ConstainsScaleInfo, sortPriority=25}
  n85 = Add(
   A=Add(
    A=@n79.RGB
    B=MaterialFunctionCall(Vector=@n84, Quat=@n50) {materialFunction=&MultiplyVectorWithQuaternion}.Result
   ) {constB=0}
   B=@n82.RGB
  ) {constB=0}
  StaticSwitchParameter(True=Transform(None=Subtract(A=@n85, B=@n81)) {transformSourceType=TRANSFORMSOURCE_Local}, False=Constant3Vector()) {defaultValue=true, group=Deformation, parameterName=ApplyWPO, sortPriority=5}
  WorldPosition()
 ]}
